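# Translucent lit material for acid bubble Niagara mesh particles in the SprocketPoppy environment effects.
# Color blends particle color plus sky-light environment samples, one with a refraction-like bent direction, toward boosted particle color via Fresnel.
# Opacity combines that lighting, Fresnel and depth-fade intersections, minus per-particle bubble noise driven by dynamic Erode, scaled by particle alpha and camera distance fade.
# Two summed world-space sine waves, blended by a dynamic WPO parameter, form vertex wobble.
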Material /Game/Effects/Fort_Effects/Effects/Environment/Hermes/SprocketPoppy/Materials/M_SprocketPoppy_AcidBubbles
graph {
 properties {bUsedWithNiagaraMeshParticles=true, blendMode=BLEND_Translucent, pixelDepthOffsetMode=PDOM_Legacy, usedShadingModels=Default Lit}
 nodes [
  n1 = ParticleColor()
  n2 = CameraVectorWS()
  n3 = VertexNormalWS()
  n4 = Normalize(VectorInput=@n3)
  n5 = Saturate(None=Subtract(A=DotProduct(A=@n2, B=@n4)) {constB=0.15})
  n6 = Constant() {r=0.6366}
  n7 = Multiply(A=LinearInterpolate(Alpha=Power(Base=@n5, Exp=@n6) {constExponent=0.5}) {constA=1, constB=0}, B=@n4)
  n8 = LinearInterpolate(A=SkyLightEnvMapSample(Direction=Subtract(A=@n7, B=@n2) {constA=0, constB=0}), B=SkyLightEnvMapSample(Direction=@n3)) {constAlpha=0.75}
  n9 = Fresnel(Normal=VertexNormalWS()) {exponent=1}
  n10 = Power(Base=Saturate(None=Multiply(A=@n9) {constB=2})) {constExponent=1}
  LinearInterpolate(A=Add(A=@n1.RGB, B=@n8), B=Saturate(None=Multiply(A=@n1.RGB) {constB=3}), Alpha=@n10) {constAlpha=0.33, constB=0.8}
  n11 = Add(
   A=Multiply(A=Clamp(None=@n8) {minDefault=0.5}, B=Clamp(None=@n10) {minDefault=0.75}) {constA=1}
   B=OneMinus(None=DepthFade() {fadeDistanceDefault=1.5})
  )
  n12 = LinearInterpolate(Alpha=DynamicParameter() {defaultValue=<rgba 0 0 1 1>, paramNames [Erode, WPO, "", ""]}.Erode) {constA=-1, constAlpha=0.6, constB=0.2}
  n13 = TextureCoordinate()
  n14 = TextureSample(UVs=Add(A=Multiply(A=@n13), B=ParticleRandom())) {samplerType=SAMPLERTYPE_LinearGrayscale, texture=/Game/Global/Textures/Default/Noise/T_Bubble_B}
  n15 = Divide(A=Saturate(None=Add(A=@n12, B=@n14.R))) {constB=0.1}
  n16 = MaterialFunctionCall(Distance=ScalarParameter() {defaultValue=2000, parameterName=CameraDistanceFade_Far}) {materialFunction=&Near_Camera_Fade}
  Multiply(A=@n1.A, B=Multiply(A=Subtract(A=@n11, B=@n15), B=OneMinus(None=@n16.Result)))
  n17 = Constant2Vector()
  n18 = DotProduct(
   A=Normalize(VectorInput=VectorParameter() {parameterName=WaveDirection_A, defaultValue=<rgba 0 0 1 1>}.RGB)
   B=WorldPosition() {worldPositionShaderOffset=WPT_ExcludeAllShaderOffsets}.XYZ
  )
  n19 = ComponentMask(None=@n18) {r=true}
  n20 = Multiply(A=Time(), B=ScalarParameter() {defaultValue=10, parameterName=WaveSpeed1})
  n21 = Divide(A=ComponentMask(None=Add(A=@n19, B=@n20)) {r=true}, B=ScalarParameter() {defaultValue=10, parameterName=WaveLength1})
  n22 = Divide(A=Add(A=ComponentMask(None=@n18) {g=true}, B=@n20), B=ScalarParameter() {defaultValue=5, parameterName=WaveLength2})
  n23 = Multiply(A=Add(A=Sine(None=@n21), B=Sine(None=@n22)), B=ScalarParameter() {defaultValue=0.25, parameterName=Amplitude})
  NamedRerouteDeclaration(None=LinearInterpolate(A=Constant(), B=AppendVector(A=@n17, B=@n23), Alpha=DynamicParameter() {defaultValue=<rgba 0 0 1 1>, paramNames [Erode, WPO, "", ""]}.WPO)) {name=WPO, nodeColor=<rgba 0.5937 1 0 1>}
  NamedRerouteUsage()
 ]
}
