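# Unlit, masked material for Niagara mesh particles, an acid fish bone effect that reuses the Teriyaki fish skull pickaxe's diffuse texture.
# The texture is tinted by particle color and passed through the shared FX emissive adjustment function.
# A concrete crack texture serves as the dissolve ramp, driven by the Erode dynamic parameter.
# The clamped dissolve result is multiplied by particle alpha, so the mesh can erode per particle.
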Material /Game/Effects/Fort_Effects/Effects/Environment/Hermes/SprocketPoppy/Materials/M_SprocketPoppy_FishBone_Acid
graph {
 properties {
  bAutomaticallySetUsageInEditor=false
  bUsedWithNiagaraMeshParticles=true
  blendMode=BLEND_Masked
  shadingModel=MSM_Unlit
  usedShadingModels=Unlit
 }
 nodes [
  n1 = TextureCoordinate()
  n2 = ParticleColor()
  MaterialFunctionCall(Emissive=Multiply(
    A=TextureSample(UVs=@n1) {texture=/Game/Weapons/FORT_Melee/Pickaxe_Teriyaki_Fish_Skull_1h/Textures/T_Pickaxe_Teriyaki_Fish_Skull_1h_D}.RGB
    B=@n2.RGB
   )) {materialFunction=&MF_FX_EmissiveAdjustments}
  Multiply(
   A=@n2.A
   B=MaterialFunctionCall!(
    Ramp=TextureSample() {samplerType=SAMPLERTYPE_Grayscale, texture=/Game/Effects/Fort_Effects/Textures/Tile/T_Concrete_Crack_A}.RGB
    Dissolve=DynamicParameter() {defaultValue=<rgba 0 1 1 1>, paramNames [Erode,Param2,Param3,Param4]}.Erode
    EdgeSharpness
   ) {materialFunction=/Game/Effects/Fort_Effects/Materials/_Functions/MF_Dissolve}.ClampedResult
  )
 ]
}
