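# Niagara environment effect that covers an acid pool with GPU bubbles that grow and shrink away, with late-life splash and ripple sprites.
# One active emitter spawns densely over a wide, flat cylinder, then hard-coded world-space kill volumes and region masks trim particles, tying it to one map location.
# Curves drive mesh scale, material parameters and sprites; distance culling applies and splash alpha fades with distance.
# A second, similar emitter is disabled.
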NiagaraSystem /Game/Effects/Fort_Effects/Effects/Environment/Hermes/SprocketPoppy/NS_SprocketPoppy_AcidBubbles
dependencies {
 usedDataInterfaces [&NiagaraDataInterfaceCurve, &NiagaraDataInterfaceColorCurve]
 usedDynamicInputs []
 usedModules [&SystemState]
}
emitters {
 Bubbles_Acid {
  bEnabled=false
  simTarget=GPUComputeSim
  emitterUpdateScript [
   &EmitterState(Enable Distance Culling=true, MaxDistance=2000, Module.bUseMaxDistance=true, Module.bUseMinDistance=true, Scalability Mode=::NewEnumerator1)
   &SpawnRate(SpawnRate=9000, Use Spawn Probability=false)
  ]
  particleSpawnScript [
   &InitializeParticle(
    Color=<rgba 0.2909 1 0 0.2>
    Lifetime Min=1.5
    Lifetime Mode=::NewEnumerator1
    Mesh Scale Mode=::NewEnumerator1
    Mesh Uniform Scale Max=0.1
    Mesh Uniform Scale Min=0.06
    Sprite Rotation Mode=::NewEnumerator2
    UsePositionOffset=false
    Write Material Random=true
   ) {module=InitializeParticle001}
   &ShapeLocation(
    Cylinder Height=2
    Cylinder Radius=1500
    MaskSpawnGroup=false
    Override Seed=false
    Shape Primitive=::NewEnumerator1
    Surface Only=true
    UseOwnerScale=true
   ) {module=ShapeLocation002}
   &KillParticlesInVolume(
    Box Size:Vector3f=<2450 1440 100>
    Invert Volume=true
    Kill Shape=::NewEnumerator1
    Volume Origin:Vector3f=&AddVectorToPosition(Position=<14336 76286.0391 1916>, Vector=<0 0 0>)
   ) {enabled=false}
   &KillParticlesInVolume(
    Box Size:Vector3f=<2600 520 100>
    Kill Shape=::NewEnumerator1
    Volume Origin:Vector3f=&AddVectorToPosition(Position=<14336 76288.4609 1916>, Vector=<0 0 0>)
   ) {enabled=false, module=KillParticlesInVolume003}
   &KillParticlesInVolume(
    Sphere Radius=108
    Volume Origin:Vector3f=&AddVectorToPosition(Position=<15540.1465 75587.2812 1943.2439>, Vector=<0 0 0>)
   ) {enabled=false, module=KillParticlesInVolume001}
   &KillParticlesInVolume(
    Sphere Radius=140
    Volume Origin:Vector3f=&AddVectorToPosition(Position=<13113.0205 75608.9844 1943.244>, Vector=<0 0 0>)
   ) {enabled=false, module=KillParticlesInVolume002}
   &KillParticlesInVolume(
    Sphere Radius=145
    Volume Origin:Vector3f=&AddVectorToPosition(Position=<15544.9082 76990.0938 1943.244>, Vector=<0 0 0>)
   ) {enabled=false, module=KillParticlesInVolume004}
  ]
  particleUpdateScript [
   &ParticleState()
   &ScaleMeshSize(Scale Factor=&VectorFromFloat(Value=&FloatFromCurve(FloatCurve=DataInterface() {propertyValues {Curve {keys [<at 0 0.2>,<at 0.8 1>,<at 1 0>]}}})))
   &DynamicMaterialParameters(
    Index 0 Param 1=&FloatFromCurve(FloatCurve=DataInterface() {propertyValues {Curve {keys [
         {interpMode=RCIM_Cubic, leaveTangent=7.3052, tangentMode=RCTM_Break, time=0.6, value=0}
         {interpMode=RCIM_Cubic, tangentMode=RCTM_Break, time=1, value=1}
        ]}}})
    Index 0 Param 4=1.35
    Module.Index0Float0Write=true
    Module.Index0Float1Write=true
    Module.Index0Float2Write=true
    Module.Index0Float3Write=true
   ) {module=DynamicMaterialParameters001}
   &SolveForcesAndVelocity(Clamp Velocity=false, Limit Acceleration=false)
   /SetVariables(
    Particles.SplashColor:LinearColor=&MultiplyLinearColor(
     Linear Color=&RandomRangeLinearColor(Minimum=<rgba 0.1183 0.1256 0.0158 1>, Maximum=<rgba 0.1633 0.1734 0.0218 1>)
     Scale Factor=&MakeVector4(
      X=1
      Y=1
      Z=1
      W=&FloatFromCurve(
       FloatCurve=DataInterface() {propertyValues {Curve {keys [
           {interpMode=RCIM_Cubic, leaveTangent=-5.3351, tangentMode=RCTM_Break, time=0.5, value=1}
           {interpMode=RCIM_Cubic, tangentMode=RCTM_User, time=1, value=0}
          ]}}}
       CurveIndex=$Emitter.Scalability.DistanceFraction
       Scale Curve=0.1
      )
     )
    )
    Particles.SplashDynamicParamenter:Vector4f=&MakeVector4(
     X=0
     Y=0
     Z=0
     W=&FloatFromCurve(FloatCurve=DataInterface() {propertyValues {Curve {keys [
          {interpMode=RCIM_Cubic, leaveTangent=6.9175, tangentMode=RCTM_Break, time=0.7, value=0}
          {arriveTangent=0.2965, interpMode=RCIM_Cubic, leaveTangent=2.756, tangentMode=RCTM_Break, time=0.9, value=1}
         ]}}})
    )
    Particles.SplashIndex:NiagaraFloat=&RandomRangeFloat(Minimum=0, Maximum=4)
    Particles.SplashSize:Vector2f=&Vector2DFromFloat(Value=&FloatFromCurve(
      FloatCurve=DataInterface() {propertyValues {Curve {keys [<at 0 0>,<at 0.6 0>,<at 0.7 1>]}}}
      Scale Curve=&RandomRangeFloat(Minimum=30, Maximum=35)
     ))
   ) {isSetParameters=true, module=SetVariables_#1}
   /SetVariables(Particles.RippleFacing:Vector3f=<0 0 1>) {isSetParameters=true, module=SetVariables_#2}
   /SetVariables(
    Particles.Ripple2Color:LinearColor=&MultiplyLinearColor(
     Linear Color=&ColorFromCurve(ColorCurve=DataInterface() {propertyValues {
        AlphaCurve {keys [<at 0.5818 0>,<at 0.6472 0.1>,<at 0.9661 0 RCIM_Cubic>]}
        BlueCurve {keys [<at 0.8093 0.2562 RCIM_Cubic>,<at 1 0.2034 RCIM_Cubic>]}
        GreenCurve {keys [<at 0.8093 0.8041 RCIM_Cubic>,<at 1 0.2576 RCIM_Cubic>]}
        RedCurve {keys [<at 0.8093 1 RCIM_Cubic>,<at 1 0.3542 RCIM_Cubic>]}
       }})
     Scale Factor=&MakeVector4(X=1, Y=1, Z=1, W=&RandomRangeFloat(Minimum=2, Maximum=1))
    )
    Particles.Ripple2Size:Vector2f=&Vector2DFromFloat(Value=&FloatFromCurve(
      FloatCurve=DataInterface() {propertyValues {Curve {keys [
          {interpMode=RCIM_Cubic, leaveTangent=-0.1034, tangentMode=RCTM_Break, time=0.47, value=0}
          {arriveTangent=0.1953, interpMode=RCIM_Cubic, tangentMode=RCTM_Break, time=1, value=1}
         ]}}}
      Scale Curve=&RandomRangeFloat(Minimum=20, Maximum=25)
     ))
    Particles.RippleCamera=1
   ) {isSetParameters=true, module=SetVariables_#3}
   &RegionMask(
    Region Dimensions:Vector3f=&Lerp_Vector(
     A=<4000 700 100>
     B=<4000 400 100>
     Alpha=&UniformAOrBFloat(A=0, Distribution Weight=0.1, Evaluation Type=::NewEnumerator0, Recalculate Random Each Loop=true, Override Seed=false)
    )
    Region Falloff=0
    Region Origin:Vector3f=&AddVectorToPosition(Position=<14336 75794.375 1916>, Vector=<0 0 0>)
    Region Type:ENiagaraDebugDrawShape=::NewEnumerator1
    UseOffset=false
    UseRotation=false
   ) {enabled=false}
   &RegionMask(
    Region Dimensions:Vector3f=&Lerp_Vector(
     A=<4000 700 100>
     B=<4000 400 100>
     Alpha=&UniformAOrBFloat(A=0, Distribution Weight=0.1, Evaluation Type=::NewEnumerator0, Recalculate Random Each Loop=true, Override Seed=false)
    )
    Region Falloff=0
    Region Origin:Vector3f=&AddVectorToPosition(Position=<14336 76775.9844 1916>, Vector=<0 0 0>)
    Region Type:ENiagaraDebugDrawShape=::NewEnumerator1
    UseOffset=false
    UseRotation=false
   ) {enabled=false, module=RegionMask001}
   &KillParticles(Kill Particles=&Boolean_OrOperation(Bool1=$Particles.RegionMask.IsInsideMask, Bool2=$Particles.RegionMask001.IsInsideMask)) {enabled=false}
  ]
  rendererClasses [&NiagaraSpriteRendererProperties, &NiagaraMeshRendererProperties]
  renderers [
   {rendererClass=&NiagaraSpriteRendererProperties}
   {rendererClass=&NiagaraMeshRendererProperties, rendererIndex=1}
   {rendererClass=&NiagaraSpriteRendererProperties, rendererIndex=2}
  ]
 }
 Bubbles_Acid003 {
  simTarget=GPUComputeSim
  emitterUpdateScript [
   &EmitterState(Enable Distance Culling=true, MaxDistance=2000, Module.bUseMaxDistance=true, Module.bUseMinDistance=true, Scalability Mode=::NewEnumerator1)
   &SpawnRate(SpawnRate=9000, Use Spawn Probability=false)
  ]
  particleSpawnScript [
   &InitializeParticle(
    Color=<rgba 0.2909 1 0 0.2>
    Lifetime Min=1.5
    Lifetime Mode=::NewEnumerator1
    Mesh Scale Mode=::NewEnumerator1
    Mesh Uniform Scale Max=0.1
    Mesh Uniform Scale Min=0.06
    Sprite Rotation Mode=::NewEnumerator2
    UsePositionOffset=false
    Write Material Random=true
   ) {module=InitializeParticle001}
   &ShapeLocation(
    Cylinder Height=2
    Cylinder Radius=1500
    MaskSpawnGroup=false
    Override Seed=false
    Shape Primitive=::NewEnumerator1
    Surface Only=true
    UseOwnerScale=true
   ) {module=ShapeLocation002}
   &KillParticlesInVolume(
    Box Size:Vector3f=<2450 1440 100>
    Invert Volume=true
    Kill Shape=::NewEnumerator1
    Volume Origin:Vector3f=&AddVectorToPosition(Position=<14336 76286.0391 1916>, Vector=<0 0 0>)
   )
   &KillParticlesInVolume(
    Box Size:Vector3f=<2600 520 100>
    Kill Shape=::NewEnumerator1
    Volume Origin:Vector3f=&AddVectorToPosition(Position=<14336 76288.4609 1916>, Vector=<0 0 0>)
   ) {module=KillParticlesInVolume003}
   &KillParticlesInVolume(
    Sphere Radius=108
    Volume Origin:Vector3f=&AddVectorToPosition(Position=<15540.1465 75587.2812 1943.2439>, Vector=<0 0 0>)
   ) {module=KillParticlesInVolume001}
   &KillParticlesInVolume(
    Sphere Radius=140
    Volume Origin:Vector3f=&AddVectorToPosition(Position=<13113.0205 75608.9844 1943.244>, Vector=<0 0 0>)
   ) {module=KillParticlesInVolume002}
   &KillParticlesInVolume(
    Sphere Radius=145
    Volume Origin:Vector3f=&AddVectorToPosition(Position=<15544.9082 76990.0938 1943.244>, Vector=<0 0 0>)
   ) {module=KillParticlesInVolume004}
  ]
  particleUpdateScript [
   &ParticleState()
   &ScaleMeshSize(Scale Factor=&VectorFromFloat(Value=&FloatFromCurve(FloatCurve=DataInterface() {propertyValues {Curve {keys [<at 0 0.2>,<at 0.8 1>,<at 1 0>]}}})))
   &DynamicMaterialParameters(
    Index 0 Param 1=&FloatFromCurve(FloatCurve=DataInterface() {propertyValues {Curve {keys [
         {interpMode=RCIM_Cubic, leaveTangent=7.3052, tangentMode=RCTM_Break, time=0.6, value=0}
         {interpMode=RCIM_Cubic, tangentMode=RCTM_Break, time=1, value=1}
        ]}}})
    Index 0 Param 4=1.35
    Module.Index0Float0Write=true
    Module.Index0Float1Write=true
    Module.Index0Float2Write=true
    Module.Index0Float3Write=true
   ) {module=DynamicMaterialParameters001}
   &SolveForcesAndVelocity(Clamp Velocity=false, Limit Acceleration=false)
   /SetVariables(
    Particles.SplashColor:LinearColor=&MultiplyLinearColor(
     Linear Color=&RandomRangeLinearColor(Minimum=<rgba 0.1183 0.1256 0.0158 1>, Maximum=<rgba 0.1633 0.1734 0.0218 1>)
     Scale Factor=&MakeVector4(
      X=1
      Y=1
      Z=1
      W=&FloatFromCurve(
       FloatCurve=DataInterface() {propertyValues {Curve {keys [
           {interpMode=RCIM_Cubic, leaveTangent=-5.3351, tangentMode=RCTM_Break, time=0.5, value=1}
           {interpMode=RCIM_Cubic, tangentMode=RCTM_User, time=1, value=0}
          ]}}}
       CurveIndex=$Emitter.Scalability.DistanceFraction
       Scale Curve=0.1
      )
     )
    )
    Particles.SplashDynamicParamenter:Vector4f=&MakeVector4(
     X=0
     Y=0
     Z=0
     W=&FloatFromCurve(FloatCurve=DataInterface() {propertyValues {Curve {keys [
          {interpMode=RCIM_Cubic, leaveTangent=6.9175, tangentMode=RCTM_Break, time=0.7, value=0}
          {arriveTangent=0.2965, interpMode=RCIM_Cubic, leaveTangent=2.756, tangentMode=RCTM_Break, time=0.9, value=1}
         ]}}})
    )
    Particles.SplashIndex:NiagaraFloat=&RandomRangeFloat(Minimum=0, Maximum=4)
    Particles.SplashSize:Vector2f=&Vector2DFromFloat(Value=&FloatFromCurve(
      FloatCurve=DataInterface() {propertyValues {Curve {keys [<at 0 0>,<at 0.6 0>,<at 0.7 1>]}}}
      Scale Curve=&RandomRangeFloat(Minimum=30, Maximum=35)
     ))
   ) {isSetParameters=true, module=SetVariables_#1}
   /SetVariables(Particles.RippleFacing:Vector3f=<0 0 1>) {isSetParameters=true, module=SetVariables_#2}
   /SetVariables(
    Particles.Ripple2Color:LinearColor=&MultiplyLinearColor(
     Linear Color=&ColorFromCurve(ColorCurve=DataInterface() {propertyValues {
        AlphaCurve {keys [<at 0.5818 0>,<at 0.6472 0.1>,<at 0.9661 0 RCIM_Cubic>]}
        BlueCurve {keys [<at 0.8093 0.2562 RCIM_Cubic>,<at 1 0.2034 RCIM_Cubic>]}
        GreenCurve {keys [<at 0.8093 0.8041 RCIM_Cubic>,<at 1 0.2576 RCIM_Cubic>]}
        RedCurve {keys [<at 0.8093 1 RCIM_Cubic>,<at 1 0.3542 RCIM_Cubic>]}
       }})
     Scale Factor=&MakeVector4(X=1, Y=1, Z=1, W=&RandomRangeFloat(Minimum=2, Maximum=1))
    )
    Particles.Ripple2Size:Vector2f=&Vector2DFromFloat(Value=&FloatFromCurve(
      FloatCurve=DataInterface() {propertyValues {Curve {keys [
          {interpMode=RCIM_Cubic, leaveTangent=-0.1034, tangentMode=RCTM_Break, time=0.47, value=0}
          {arriveTangent=0.1953, interpMode=RCIM_Cubic, tangentMode=RCTM_Break, time=1, value=1}
         ]}}}
      Scale Curve=&RandomRangeFloat(Minimum=20, Maximum=25)
     ))
    Particles.RippleCamera=1
   ) {isSetParameters=true, module=SetVariables_#3}
   &RegionMask(
    Region Dimensions:Vector3f=&Lerp_Vector(
     A=<4000 700 100>
     B=<4000 400 100>
     Alpha=&UniformAOrBFloat(A=0, Distribution Weight=0.1, Evaluation Type=::NewEnumerator0, Recalculate Random Each Loop=true, Override Seed=false)
    )
    Region Falloff=0
    Region Origin:Vector3f=&AddVectorToPosition(Position=<14336 75794.375 1916>, Vector=<0 0 0>)
    Region Type:ENiagaraDebugDrawShape=::NewEnumerator1
    UseOffset=false
    UseRotation=false
   )
   &RegionMask(
    Region Dimensions:Vector3f=&Lerp_Vector(
     A=<4000 700 100>
     B=<4000 400 100>
     Alpha=&UniformAOrBFloat(A=0, Distribution Weight=0.1, Evaluation Type=::NewEnumerator0, Recalculate Random Each Loop=true, Override Seed=false)
    )
    Region Falloff=0
    Region Origin:Vector3f=&AddVectorToPosition(Position=<14336 76775.9844 1916>, Vector=<0 0 0>)
    Region Type:ENiagaraDebugDrawShape=::NewEnumerator1
    UseOffset=false
    UseRotation=false
   ) {module=RegionMask001}
   &KillParticles(Kill Particles=&Boolean_OrOperation(Bool1=$Particles.RegionMask.IsInsideMask, Bool2=$Particles.RegionMask001.IsInsideMask))
   &ScaleColor(
    Scale Alpha=&FloatFromCurve(
     FloatCurve=DataInterface() {propertyValues {Curve {keys [{interpMode=RCIM_Cubic, tangentMode=RCTM_User, time=0.1, value=1}, {interpMode=RCIM_Cubic, tangentMode=RCTM_User, time=0.15, value=0}]}}}
     CurveIndex=$Emitter.Scalability.DistanceFraction
    )
    ScaleA=true
    ScaleRGB=false
   ) {enabled=false, module=ScaleColor001}
   &Color(Color=&LerpLinearColors(StartColor=<rgba 3 0.0667 0.4318 1>, EndColor=<rgba 0.1841 2.8528 3 1>, LerpFactor=$Particles.RegionMask.Float)) {enabled=false}
  ]
  rendererClasses [&NiagaraSpriteRendererProperties, &NiagaraMeshRendererProperties]
  renderers [
   {rendererClass=&NiagaraSpriteRendererProperties}
   {rendererClass=&NiagaraMeshRendererProperties, rendererIndex=1}
   {rendererClass=&NiagaraSpriteRendererProperties, rendererIndex=2}
  ]
 }
}
summary {
 systemName=NS_SprocketPoppy_AcidBubbles
 emitters [
  {
   emitterName=Bubbles_Acid003
   simTarget=GPUComputeSim
   rendererClasses [&NiagaraSpriteRendererProperties, &NiagaraMeshRendererProperties]
  }
  {
   bEnabled=false
   emitterName=Bubbles_Acid
   simTarget=GPUComputeSim
   rendererClasses [&NiagaraSpriteRendererProperties, &NiagaraMeshRendererProperties]
  }
 ]
 userVariables []
}
