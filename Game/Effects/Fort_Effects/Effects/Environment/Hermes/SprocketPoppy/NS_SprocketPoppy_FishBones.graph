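# Niagara environment effect that floats fish-bone mesh particles over a fixed area, then sinks and darkens them late in life.
# The GPU emitter layers an opening surge over steady spawning on a cylinder, using hard-coded world-space kill boxes to trim spawns into strips.
# Curl noise, rising drag and late downward gravity drive motion, while curves animate scale, color and a dynamic material parameter.
# A near-identical emitter at another location is disabled.
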NiagaraSystem /Game/Effects/Fort_Effects/Effects/Environment/Hermes/SprocketPoppy/NS_SprocketPoppy_FishBones
dependencies {usedDataInterfaces [&NiagaraDataInterfaceCurve], usedDynamicInputs [], usedModules [&SystemState]}
emitters {
 FishBones_Floating {
  bEnabled=false
  simTarget=GPUComputeSim
  emitterUpdateScript [
   &EmitterState()
   &SpawnRate(
    SpawnRate=&FloatFromCurve(
     FloatCurve=DataInterface() {propertyValues {Curve {keys [
         {interpMode=RCIM_Cubic, tangentMode=RCTM_User, time=0, value=0}
         {arriveTangent=10.0781, interpMode=RCIM_Cubic, leaveTangent=-9.3093, tangentMode=RCTM_Break, time=0.25, value=1}
         {tangentMode=RCTM_User, time=0.5, value=0}
        ]}}}
     CurveIndex=$Engine.System.Age
     Scale Curve=55
    )
    Use Spawn Probability=false
   )
   &SpawnRate(SpawnRate=&RandomRangeFloat(Minimum=25, Maximum=20), Use Spawn Probability=false) {module=SpawnRate002}
  ]
  particleSpawnScript [
   &InitializeParticle(
    Color Maximum=<rgba 0.2876 0.3734 0.2876 1>
    Color Minimum=<rgba 0.1451 0.2169 0.1451 1>
    Color Mode=::NewEnumerator2
    Lifetime Max=6
    Lifetime Min=5
    Lifetime Mode=::NewEnumerator1
    Mesh Scale Max=<1 1 1>
    Mesh Scale Min=<0.5 0.95 0.95>
    Mesh Scale Mode=::NewEnumerator2
    Sprite Size Mode=::NewEnumerator3
    Uniform Sprite Size=80
    UsePositionOffset=false
    Write Material Random=true
   )
   &ShapeLocation(
    Cylinder Height=20
    Cylinder Radius=1500
    MaskSpawnGroup=false
    Override Seed=false
    Shape Primitive=::NewEnumerator1
    Surface Only=true
    Surface Only Band Thickness=0.8
    Use Endcaps In Surface Only Mode=false
    UseOwnerScale=true
   ) {module=ShapeLocation001}
   &ShapeLocation(
    Cylinder Height=10
    Cylinder Radius=1500
    MaskSpawnGroup=false
    Override Seed=false
    Shape Primitive=::NewEnumerator1
    Surface Only=true
    Surface Only Band Thickness=0.8
    Use Endcaps In Surface Only Mode=false
    UseOwnerScale=true
   ) {enabled=false}
   &ApplyInitialForces()
   &KillParticlesInVolume(
    Box Size:Vector3f=<2400 1400 100>
    Invert Volume=true
    Kill Shape=::NewEnumerator1
    Volume Origin:Vector3f=&AddVectorToPosition(Position=<4 -12804 12>, Vector=<0 0 0>)
   ) {module=KillParticlesInVolume002}
   &KillParticlesInVolume(
    Box Size:Vector3f=<2500 550 100>
    Kill Shape=::NewEnumerator1
    Volume Origin:Vector3f=&AddVectorToPosition(Position=<4 -12798 12>, Vector=<0 0 0>)
   ) {module=KillParticlesInVolume003}
   &InitialMeshOrientation(
    Mesh Orientation Mode=::NewEnumerator7
    Rotation=&RandomRangeVector(Minimum=<-0.01 0 0>, Maximum=<0.01 0 1>)
    Use Rotation Vector=true
   )
  ]
  particleUpdateScript [
   &ParticleState()
   &CurlNoiseForce(Bias Noise=false, Noise Strength=5, Pan Noise=false)
   &Drag(
    Drag=&FloatFromCurve(
     FloatCurve=DataInterface() {propertyValues {Curve {keys [{interpMode=RCIM_Cubic, tangentMode=RCTM_User, time=0, value=0}, {interpMode=RCIM_Cubic, tangentMode=RCTM_User, time=1, value=1}]}}}
     Scale Curve=&RandomRangeFloat(Minimum=0, Maximum=10)
    )
    Use Linear Drag=true
    Use Rotational Drag=true
   )
   &GravityForce(Gravity=&Multiply_VectorByFloat(
     Vector=<0 0 1>
     Float=&FloatFromCurve(
      FloatCurve=DataInterface() {propertyValues {Curve {keys [
          {interpMode=RCIM_Cubic, leaveTangent=4.4609, tangentMode=RCTM_Break, time=0.5, value=0}
          {interpMode=RCIM_Cubic, tangentMode=RCTM_User, time=1, value=1}
         ]}}}
      Scale Curve=&RandomRangeFloat(Minimum=-50, Maximum=-70)
     )
    ))
   &SolveForcesAndVelocity(Clamp Velocity=false, Limit Acceleration=false)
   &ScaleMeshSize(Scale Factor=&VectorFromFloat(Value=&FloatFromCurve(FloatCurve=DataInterface() {propertyValues {Curve {keys [
          {interpMode=RCIM_Cubic, tangentMode=RCTM_Break, time=0, value=0}
          {arriveTangent=1, interpMode=RCIM_Cubic, leaveTangent=0.0495, tangentMode=RCTM_Break, time=0.05, value=1}
          {interpMode=RCIM_Cubic, leaveTangent=-5.5791, tangentMode=RCTM_Break, time=1, value=1}
         ]}}}))) {module=ScaleMeshSize002}
   &ScaleColor(
    Scale RGB=&MakeVectorFromLinearColorRGB(Color=&LerpLinearColors(
      StartColor=<rgba 1 1 1 1>
      EndColor=<rgba 0 0 0 1>
      LerpFactor=&FloatFromCurve(FloatCurve=DataInterface() {propertyValues {Curve {keys [
           <at 0 0.7 RCIM_Cubic>
           {arriveTangent=-1.6667, interpMode=RCIM_Cubic, leaveTangent=-0.1024, tangentMode=RCTM_Break, time=0.2, value=0}
           {interpMode=RCIM_Cubic, tangentMode=RCTM_User, time=0.6, value=0}
           {arriveTangent=3.3194, interpMode=RCIM_Cubic, leaveTangent=3.3194, tangentMode=RCTM_User, time=1, value=0.8}
          ]}}})
     ))
    ScaleA=false
    ScaleRGB=true
   ) {module=ScaleColor001}
   &DynamicMaterialParameters(
    Index 0 Param 1=&FloatFromCurve(FloatCurve=DataInterface() {propertyValues {Curve {keys [
         {interpMode=RCIM_Cubic, tangentMode=RCTM_Break, time=0.85, value=0}
         {arriveTangent=14.1798, interpMode=RCIM_Cubic, tangentMode=RCTM_Break, time=1, value=1}
        ]}}})
    Module.Index0Float0Write=true
    Module.Index0Float1Write=true
    Module.Index0Float2Write=true
    Module.Index0Float3Write=true
   )
  ]
  rendererClasses [&NiagaraMeshRendererProperties]
  renderers [{rendererClass=&NiagaraMeshRendererProperties, rendererIndex=0}]
 }
 FishBones_Floating001 {
  simTarget=GPUComputeSim
  emitterUpdateScript [
   &EmitterState()
   &SpawnRate(
    SpawnRate=&FloatFromCurve(
     FloatCurve=DataInterface() {propertyValues {Curve {keys [
         {interpMode=RCIM_Cubic, tangentMode=RCTM_User, time=0, value=0}
         {arriveTangent=10.0781, interpMode=RCIM_Cubic, leaveTangent=-9.3093, tangentMode=RCTM_Break, time=0.25, value=1}
         {tangentMode=RCTM_User, time=0.5, value=0}
        ]}}}
     CurveIndex=$Engine.System.Age
     Scale Curve=55
    )
    Use Spawn Probability=false
   )
   &SpawnRate(SpawnRate=&RandomRangeFloat(Minimum=25, Maximum=20), Use Spawn Probability=false) {module=SpawnRate002}
  ]
  particleSpawnScript [
   &InitializeParticle(
    Color Maximum=<rgba 0.2876 0.3734 0.2876 1>
    Color Minimum=<rgba 0.1451 0.2169 0.1451 1>
    Color Mode=::NewEnumerator2
    Lifetime Max=6
    Lifetime Min=5
    Lifetime Mode=::NewEnumerator1
    Mesh Scale Max=<1 1 1>
    Mesh Scale Min=<0.5 0.95 0.95>
    Mesh Scale Mode=::NewEnumerator2
    Sprite Size Mode=::NewEnumerator3
    Uniform Sprite Size=80
    UsePositionOffset=false
    Write Material Random=true
   )
   &ShapeLocation(
    Cylinder Height=20
    Cylinder Radius=1500
    MaskSpawnGroup=false
    Offset=<0 0 -7>
    Offset Mode=::NewEnumerator0
    Override Seed=false
    Shape Primitive=::NewEnumerator1
    Surface Only=true
    Surface Only Band Thickness=0.8
    Use Endcaps In Surface Only Mode=false
    UseOwnerScale=true
   ) {module=ShapeLocation001}
   &ApplyInitialForces()
   &KillParticlesInVolume(
    Box Size:Vector3f=<2400.7839 1400.4581 100>
    Invert Volume=true
    Kill Shape=::NewEnumerator1
    Volume Origin:Vector3f=&AddVectorToPosition(Position=<14336 76286.0391 1916>, Vector=<0 0 0>)
   ) {module=KillParticlesInVolume002}
   &KillParticlesInVolume(
    Box Size:Vector3f=<2600 500.0595 100>
    Kill Shape=::NewEnumerator1
    Volume Origin:Vector3f=&AddVectorToPosition(Position=<14336 76288.4609 1916>, Vector=<0 0 0>)
   ) {module=KillParticlesInVolume003}
   &InitialMeshOrientation(
    Mesh Orientation Mode=::NewEnumerator7
    Rotation=&RandomRangeVector(Minimum=<-0.01 0 0>, Maximum=<0.01 0 1>)
    Use Rotation Vector=true
   )
  ]
  particleUpdateScript [
   &ParticleState()
   &CurlNoiseForce(Bias Noise=false, Noise Strength=5, Pan Noise=false)
   &Drag(
    Drag=&FloatFromCurve(
     FloatCurve=DataInterface() {propertyValues {Curve {keys [{interpMode=RCIM_Cubic, tangentMode=RCTM_User, time=0, value=0}, {interpMode=RCIM_Cubic, tangentMode=RCTM_User, time=1, value=1}]}}}
     Scale Curve=&RandomRangeFloat(Minimum=0, Maximum=10)
    )
    Use Linear Drag=true
    Use Rotational Drag=true
   )
   &GravityForce(Gravity=&Multiply_VectorByFloat(
     Vector=<0 0 1>
     Float=&FloatFromCurve(
      FloatCurve=DataInterface() {propertyValues {Curve {keys [
          {interpMode=RCIM_Cubic, leaveTangent=4.4609, tangentMode=RCTM_Break, time=0.5, value=0}
          {interpMode=RCIM_Cubic, tangentMode=RCTM_User, time=1, value=1}
         ]}}}
      Scale Curve=&RandomRangeFloat(Minimum=-50, Maximum=-70)
     )
    ))
   &SolveForcesAndVelocity(Clamp Velocity=false, Limit Acceleration=false)
   &ScaleMeshSize(Scale Factor=&VectorFromFloat(Value=&FloatFromCurve(FloatCurve=DataInterface() {propertyValues {Curve {keys [
          {interpMode=RCIM_Cubic, tangentMode=RCTM_Break, time=0, value=0}
          {arriveTangent=1, interpMode=RCIM_Cubic, leaveTangent=0.0495, tangentMode=RCTM_Break, time=0.05, value=1}
          {interpMode=RCIM_Cubic, leaveTangent=-5.5791, tangentMode=RCTM_Break, time=1, value=1}
         ]}}}))) {module=ScaleMeshSize002}
   &ScaleColor(
    Scale RGB=&MakeVectorFromLinearColorRGB(Color=&LerpLinearColors(
      StartColor=<rgba 1 1 1 1>
      EndColor=<rgba 0 0 0 1>
      LerpFactor=&FloatFromCurve(FloatCurve=DataInterface() {propertyValues {Curve {keys [
           <at 0 0.7 RCIM_Cubic>
           {arriveTangent=-1.6667, interpMode=RCIM_Cubic, leaveTangent=-0.1024, tangentMode=RCTM_Break, time=0.2, value=0}
           {interpMode=RCIM_Cubic, tangentMode=RCTM_User, time=0.6, value=0}
           {arriveTangent=3.3194, interpMode=RCIM_Cubic, leaveTangent=3.3194, tangentMode=RCTM_User, time=1, value=0.8}
          ]}}})
     ))
    ScaleA=false
    ScaleRGB=true
   ) {module=ScaleColor001}
   &DynamicMaterialParameters(
    Index 0 Param 1=&FloatFromCurve(FloatCurve=DataInterface() {propertyValues {Curve {keys [
         {interpMode=RCIM_Cubic, tangentMode=RCTM_Break, time=0.85, value=0}
         {arriveTangent=14.1798, interpMode=RCIM_Cubic, tangentMode=RCTM_Break, time=1, value=1}
        ]}}})
    Module.Index0Float0Write=true
    Module.Index0Float1Write=true
    Module.Index0Float2Write=true
    Module.Index0Float3Write=true
   )
  ]
  rendererClasses [&NiagaraMeshRendererProperties]
  renderers [{rendererClass=&NiagaraMeshRendererProperties, rendererIndex=0}]
 }
}
summary {
 systemName=NS_SprocketPoppy_FishBones
 emitters [
  {bEnabled=false, emitterName=FishBones_Floating, simTarget=GPUComputeSim, rendererClasses [&NiagaraMeshRendererProperties]}
  {emitterName=FishBones_Floating001, simTarget=GPUComputeSim, rendererClasses [&NiagaraMeshRendererProperties]}
 ]
 userVariables []
}
